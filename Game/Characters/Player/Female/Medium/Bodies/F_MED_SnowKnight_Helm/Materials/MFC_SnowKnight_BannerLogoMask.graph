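# Material function for the SnowKnight helm that outputs BannerAlpha, a placeable alpha mask for a banner logo.
# Switchable UV0 or UV1 coordinates (UV1 by default) are offset, rotated and center-scaled by Banner1 parameters to sample a clamped banner shape texture.
# The texture's red channel is multiplied by the scale function's 0-1 mask.
# A Min node, fed through named reroutes, combines the logo mask with the LogoAlpha parameter.
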MaterialFunction /Game/Characters/Player/Female/Medium/Bodies/F_MED_SnowKnight_Helm/Materials/MFC_SnowKnight_BannerLogoMask
graph {nodes [
  FunctionOutput(None=Min(A=NamedRerouteUsage(), B=NamedRerouteUsage())) {outputName=BannerAlpha, sortPriority=1}
  n1 = StaticSwitchParameter(True=TextureCoordinate() {coordinateIndex=1}, False=TextureCoordinate()) {defaultValue=true, group=01_SnowKnight_BannerLogo, parameterName=UseUV1, sortPriority=1}
  n2 = AppendVector(
   A=ScalarParameter() {group=01_SnowKnight_BannerLogo, parameterName=Banner1_X}
   B=ScalarParameter() {group=01_SnowKnight_BannerLogo, parameterName=Banner1_Y}
  )
  n3 = MaterialFunctionCall(
   UVs=MaterialFunctionCall(UVs=Add(A=@n1, B=@n2), "Rotation Angle (0-1)"=ScalarParameter() {group=01_SnowKnight_BannerLogo, parameterName=Banner1_Rotation}) {materialFunction=&CustomRotator}.Rotated Values
   Texture Scale=ScalarParameter() {defaultValue=0.2, group=01_SnowKnight_BannerLogo, parameterName=Banner1_Scale}
  ) {materialFunction=&ScaleUVsByCenter}
  n4 = TextureSampleParameter2D(UVs=@n3.UVs) {
   group=01_SnowKnight_BannerLogo
   parameterName=Banner_Texture
   samplerSource=SSM_Clamp_WorldGroupSettings
   sortPriority=2
   texture=/Game/UI/Foundation/Textures/Banner/Shapes/banner_shape_standard_Mipmaps
  }
  NamedRerouteDeclaration(None=Multiply(A=@n4.R, B=@n3.0-1 mask)) {name=BannerLogo, nodeColor=<rgba 1 0 0.8672 1>}
  NamedRerouteDeclaration(None=ScalarParameter() {defaultValue=1, group=01_SnowKnight_BannerLogo, parameterName=LogoAlpha, sliderMax=1}) {name=LogoColor_Alpha, nodeColor=<rgba 1 0 0.4453 1>}
 ]}
