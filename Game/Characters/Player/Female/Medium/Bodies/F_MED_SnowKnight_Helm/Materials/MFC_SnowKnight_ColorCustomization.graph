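# Material function that adds color customization to the Snow Knight character's materials, behind a static switch.
# A channel-selectable mask texture, plus banner and chainmail inputs, defines primary, secondary, trim and crystal color regions, with chainmail boosts and chipping highlights.
# The tint combines with base color through an overlay-to-multiply blend and rewrites the incoming material attributes.
# Optional eye glow drives emissive from crystal color and an FX mask; a packed float4 mask is also output.
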MaterialFunction /Game/Characters/Player/Female/Medium/Bodies/F_MED_SnowKnight_Helm/Materials/MFC_SnowKnight_ColorCustomization
graph {nodes [
  FunctionOutput(None=StaticSwitchParameter(True=NamedRerouteUsage(), False=NamedRerouteUsage()) {group=01_SnowKnight_ColorCustomization, parameterName=UseSnowKnight_ColorCustomization, sortPriority=0})
  FunctionOutput(None=MaterialFunctionCall(X=NamedRerouteUsage(), Y=NamedRerouteUsage(), Z=NamedRerouteUsage(), A=NamedRerouteUsage()) {materialFunction=&MakeFloat4}.Result) {outputName=Mask, sortPriority=1}
  n1 = VectorParameter() {group=01_SnowKnight_ColorCustomization, parameterName=Primary_Color, sortPriority=10, defaultValue=<rgba 1 1 1 1>}
  LinearInterpolate(A=NamedRerouteUsage(), B=@n1.RGB, Alpha=NamedRerouteUsage())
  n2 = NamedRerouteUsage()
  Multiply(A=NamedRerouteUsage(), B=@n2)
  n3 = GetMaterialAttributes(None=FunctionInput() {inputName=MA, inputType=FunctionInput_MaterialAttributes}) {attributeGetTypes [#1, #2, #3]}
  NamedRerouteDeclaration(None=@n3.MaterialAttributes) {name=MA_In, nodeColor=<rgba 0 0.3437 1 1>}
  n4 = NamedRerouteUsage()
  n5 = NamedRerouteUsage()
  n6 = MaterialFunctionCall(Base=@n4, Blend=@n5) {materialFunction=&Blend_Overlay}
  n7 = Multiply(A=@n4, B=@n5)
  n8 = ScalarParameter() {group=01_SnowKnight_ColorCustomization, parameterName=OverlayTintBlend, sliderMax=1}
  n9 = Constant()
  NamedRerouteDeclaration(None=SetMaterialAttributes!(
    MaterialAttributes=NamedRerouteUsage()
    BaseColor=LinearInterpolate(A=@n6.Result, B=@n7, Alpha=@n8)
    EmissiveColor=StaticSwitchParameter(True=@n2, False=@n9) {group=01_SnowKnight_ColorCustomization, parameterName=UseEyeGlow}
   ) {attributeSetTypes [#1, #2]}) {name=MA_Out, nodeColor=<rgba 1 0.3008 0.2809 1>}
  n10 = ChannelMaskParameter(None=NamedRerouteDeclaration(None=FunctionInput() {inputName=FX, sortPriority=1}) {name=FX, nodeColor=<rgba 0.8281 0 1 1>}) {group=01_SnowKnight_ColorCustomization, maskChannel=Blue, parameterName=FX_EyeMask, sortPriority=8, defaultValue=<rgba 0 0 1 0>}
  NamedRerouteDeclaration(None=@n10.R) {name=EyeGlowMask, nodeColor=<rgba 0 0.0391 1 1>}
  n11 = VectorParameter() {group=01_SnowKnight_ColorCustomization, parameterName=Crystal_Color, sortPriority=12, defaultValue=<rgba 0 0 1 1>}
  n12 = NamedRerouteUsage()
  NamedRerouteDeclaration(None=Multiply(
    A=Multiply(A=@n11.RGB, B=@n12)
    B=ScalarParameter() {defaultValue=5, group=01_SnowKnight_ColorCustomization, parameterName=EyeGlowEmissiveBrightness}
   )) {name=Emissive_Out, nodeColor=<rgba 1 0.5859 0 1>}
  NamedRerouteDeclaration(None=ComponentMask(None=NamedRerouteUsage()) {g=true}) {name=Chipping_Mask, nodeColor=<rgba 0.8984 0 1 1>}
  NamedRerouteDeclaration(None=FunctionInput() {bUsePreviewValueAsDefault=true, inputName=BannerMask, inputType=FunctionInput_Scalar, sortPriority=5}) {name=BannerAlpha, nodeColor=<rgba 1 0.9141 0 1>}
  NamedRerouteDeclaration(None=@n3.Metallic) {name=Metallic_In, nodeColor=<rgba 0.125 0 1 1>}
  NamedRerouteDeclaration(None=FunctionInput() {bUsePreviewValueAsDefault=true, inputName=ChainmailMask, inputType=FunctionInput_Scalar, sortPriority=6}) {name=ChainmailMask, nodeColor=<rgba 0 0.9297 1 1>}
  n13 = Saturate(None=Add(A=NamedRerouteUsage(), B=NamedRerouteUsage()))
  n14 = ScalarParameter() {group=01_SnowKnight_ColorCustomization, parameterName=BaseColorBias}
  NamedRerouteDeclaration(None=Add(A=Multiply(A=@n13, B=@n14), B=NamedRerouteUsage())) {name=BaseColorModified, nodeColor=<rgba 0 0.2734 1 1>}
  n15 = TextureSampleParameter2D() {
   group=01_SnowKnight_ColorCustomization
   parameterName=Masks
   samplerType=SAMPLERTYPE_Masks
   sortPriority=1
   texture=&T_BlackMask
  }
  n16 = NamedRerouteDeclaration(None=@n15.RGB) {name=MasksRGB, nodeColor=<rgba 1 0 0.0703 1>}
  n17 = ChannelMaskParameter(None=@n16) {
   group=01_SnowKnight_ColorCustomization
   maskChannel=Green
   parameterName=TertiaryColor_ChannelMask
   sortPriority=6
   defaultValue=<rgba 0 1 0 0>
  }
  n18 = Max(
   A=NamedRerouteDeclaration(None=ChannelMaskParameter(None=@n16) {group=01_SnowKnight_ColorCustomization, parameterName=SecondaryColor_ChannelMask, sortPriority=5}.R) {name=SecondaryColor_Mask, nodeColor=<rgba 0.5938 0 1 1>}
   B=NamedRerouteDeclaration(None=@n17.R) {name=TertiaryColor_Mask, nodeColor=<rgba 0.5938 0 1 1>}
  )
  n19 = ChannelMaskParameter(None=@n16) {
   group=01_SnowKnight_ColorCustomization
   maskChannel=Blue
   parameterName=CrystalColor_ChannelMask
   sortPriority=7
   defaultValue=<rgba 0 0 1 0>
  }
  n20 = NamedRerouteDeclaration(None=@n19.R) {name=CrystalColor_Mask, nodeColor=<rgba 1 0.4219 0 1>}
  n21 = NamedRerouteUsage()
  n22 = NamedRerouteDeclaration(None=Saturate(None=Add(A=@n20, B=@n21))) {name=CrystalBannerColor_Mask, nodeColor=<rgba 0.5938 0 1 1>}
  NamedRerouteDeclaration(None=OneMinus(None=Max(A=@n18, B=@n22))) {name=PrimaryColor_Mask, nodeColor=<rgba 0.5938 0 1 1>}
  NamedRerouteDeclaration(None=@n3.BaseColor) {name=BaseColor_In, nodeColor=<rgba 0 0.3438 1 1>}
  n23 = NamedRerouteUsage()
  n24 = ScalarParameter() {parameterName=Primary_ChainmailBoost}
  n25 = Multiply(A=@n1.RGB, B=Add(B=Multiply(A=@n23, B=@n24)) {constA=1})
  n26 = VectorParameter() {group=01_SnowKnight_ColorCustomization, parameterName=Secondary_Color, sortPriority=11, defaultValue=<rgba 1 0 0 1>}
  n27 = NamedRerouteUsage()
  n28 = ScalarParameter() {parameterName=Secondary_ChainmailBoost}
  n29 = Multiply(A=@n26.RGB, B=Add(B=Multiply(A=@n27, B=@n28)) {constA=1})
  n30 = NamedRerouteUsage()
  n31 = VectorParameter() {group=01_SnowKnight_ColorCustomization, parameterName=Trim_Color, sortPriority=11, defaultValue=<rgba 0 1 0 1>}
  n32 = LinearInterpolate(
   A=LinearInterpolate(A=LinearInterpolate(A=@n25, B=@n29, Alpha=@n30), B=@n31.RGB, Alpha=NamedRerouteUsage())
   B=@n11.RGB
   Alpha=NamedRerouteUsage()
  )
  n33 = Add(A=@n32, B=ScalarParameter() {defaultValue=0.125, group=01_SnowKnight_ColorCustomization, parameterName=ChippingIntensity, sliderMax=1})
  NamedRerouteDeclaration(None=LinearInterpolate(A=@n32, B=Saturate(None=@n33), Alpha=NamedRerouteUsage()) {constB=0.5}) {name=BaseColor_Out, nodeColor=<rgba 1 0.5859 0 1>}
  NamedRerouteDeclaration(None=@n3.EmissiveColor) {name=Emissive_In, nodeColor=<rgba 0 0.3438 1 1>}
  NamedRerouteUsage()
 ]}
