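# Material function that layers an ice crystal look over an incoming material for the SnowKnight character, blended by an Alpha input.
# A world-normal Fresnel indexes a gradient texture for color and emissive, and an EdgeMap input tints edges.
# A tiled detail normal is flattened by intensity and angle-corrected blended with the base normal, with a separate shading-path input for mobile.
# The ice layer uses Subsurface Profile shading and parameterized metallic, roughness and opacity.
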MaterialFunction /Game/Characters/Player/Female/Medium/Bodies/F_MED_SnowKnight_Helm/Materials/MF_SnowKnight_Ice
graph {nodes [
  n1 = NamedRerouteUsage()
  n2 = NamedRerouteUsage()
  FeatureLevelSwitch(Default=@n1, ES3_1=@n2)
  n3 = SetMaterialAttributes!(
   MaterialAttributes
   BaseColor=NamedRerouteUsage()
   Metallic=ScalarParameter() {defaultValue=0.4, group=Ice, parameterName=IceMetallic}
   Specular=NamedRerouteUsage()
   Roughness=ScalarParameter() {defaultValue=0.05, group=Ice, parameterName=IceRoughness}
   EmissiveColor=NamedRerouteUsage()
   Opacity=ScalarParameter() {defaultValue=1, group=Ice, parameterName=Ice Crystal Opacity}
   Normal=ShadingPathSwitch(Default=@n1, Deferred=@n1, Mobile=@n2)
   ShadingModel=ShadingModel() {shadingModel=MSM_SubsurfaceProfile}
  ) {attributeSetTypes [
    #1
    #2
    #3
    #4
    #5
    #6
    #7
    #8
   ]}
  FunctionOutput(None=BlendMaterialAttributes(
    A=NamedRerouteUsage()
    B=@n3
    Alpha=FunctionInput() {bUsePreviewValueAsDefault=true, inputName=Alpha, inputType=FunctionInput_Scalar, sortPriority=1}
   ))
  MaterialFunctionCall!(Low _ PC Low and Med, "High _ XB1, PS4", "Epic _ PC High, XSX, PS5", --------, "Mobile Low _ PC Perf, Android", Mobile Med _ Switch) {materialFunction=/Game/Global/MaterialFunctions/TechArt/MF_Fortnite_QualitySwitch_MaterialAttributes}
  n4 = VectorParameter() {group=Ice, parameterName=Ice Edge Tint, defaultValue=<rgba 0.5876 0.6719 0.6652 1>}
  n5 = FunctionInput() {inputName=EdgeMap, inputType=FunctionInput_Scalar}
  n6 = ScalarParameter() {defaultValue=1, group=Ice, parameterName=Ice Edge Exponent}
  n7 = Saturate(None=Multiply(A=Power(Base=@n5, Exp=@n6), B=ScalarParameter() {defaultValue=1, group=Ice, parameterName=Ice Edge Strength}))
  n8 = Fresnel(ExponentIn=ScalarParameter() {defaultValue=0.8, group=Ice, parameterName=Crystal_FresEX}, Normal=PixelNormalWS()) {baseReflectFraction=0}
  n9 = TextureSample(UVs=MaterialFunctionCall(X=Saturate(None=@n8), Y=Constant() {r=0.75}) {materialFunction=&MakeFloat2}.Result) {
   samplerSource=SSM_Wrap_WorldGroupSettings
   texture=/Game/Characters/Player/Female/Medium/Bodies/F_MED_SnowKnight_Helm/Materials/AdvanceMaterials/T_SnowKnight_Gradients
  }
  n10 = Multiply(A=@n9.RGB, B=ScalarParameter() {defaultValue=1, group=Ice, parameterName=ice Brightness})
  NamedRerouteDeclaration(None=LinearInterpolate(
    A=@n10
    B=LinearInterpolate(A=@n10, B=@n4.RGB, Alpha=@n7)
    Alpha=ScalarParameter() {defaultValue=1, group=Ice, parameterName=Ice Use Edge Map}
   )) {name=BaseColor, nodeColor=<rgba 1 0.3516 0 1>}
  n11 = ScalarParameter() {group=Ice, parameterName=Ice Emissive Brightness}
  n12 = LinearInterpolate(B=NamedRerouteUsage(), Alpha=ScalarParameter() {defaultValue=1, group=Ice, parameterName=CenterFresnelStrength}) {constA=1}
  NamedRerouteDeclaration(None=Multiply(A=@n10, B=Multiply(A=@n11, B=@n12))) {name=Emissive, nodeColor=<rgba 0.7578 0 1 1>}
  n13 = FunctionInput() {inputName=Material, inputType=FunctionInput_MaterialAttributes}
  n14 = GetMaterialAttributes(None=@n13) {attributeGetTypes [#7, #3]}
  NamedRerouteDeclaration(None=@n14.Normal) {name=InNormal, nodeColor=<rgba 0.125 1 0 1>}
  NamedRerouteDeclaration(None=@n14.Specular) {name=InSpec, nodeColor=<rgba 0.3828 1 0 1>}
  NamedRerouteDeclaration(None=@n13) {name=InMaterial, nodeColor=<rgba 1 0.3281 0 1>}
  NamedRerouteDeclaration(None=Power(Base=@n8, Exp=ScalarParameter() {defaultValue=1, group=Ice, parameterName=Ice Fresnel})) {name=ModifiedFresnel, nodeColor=<rgba 0 1 0.3203 1>}
  n15 = NamedRerouteUsage()
  n16 = TextureSampleParameter2D(UVs=Multiply(A=TextureCoordinate(), B=ScalarParameter() {defaultValue=5, group=Ice, parameterName=Ice Normal Tiling})) {group=Ice, parameterName=Ice Normal Tiling Tex, samplerType=SAMPLERTYPE_Normal, texture=/Game/Weapons/WeaponSkins/Textures/StoneNormal}
  n17 = OneMinus(None=ScalarParameter() {group=Ice, parameterName=IceNormalIntensity})
  n18 = NamedRerouteDeclaration(None=MaterialFunctionCall(Normal=@n16.RGB, Flatness=Saturate(None=@n17)) {materialFunction=&FlattenNormal}.Result) {name=IceNormal, nodeColor=<rgba 0 1 0.0156 1>}
  NamedRerouteDeclaration(None=MaterialFunctionCall(Vector=MaterialFunctionCall(BaseNormal=@n15, AdditionalNormal=@n18) {materialFunction=&BlendAngleCorrectedNormals}.Result) {materialFunction=&SafeNormalize}.Result) {name=Normal, nodeColor=<rgba 0 1 0.9531 1>}
 ]}
