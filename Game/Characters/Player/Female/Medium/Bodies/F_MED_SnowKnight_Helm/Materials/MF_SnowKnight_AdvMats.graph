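# Material function that layers the Snow Knight character's special surface effects onto an incoming material attributes input.
# It chains the ice function into the galaxy function, then into a thin-film function driven by Fresnel.
# Ice, galaxy and chroma alphas are each built from primary and secondary masks weighted by scalar parameters, then summed and saturated.
# A side branch that lerps the galaxy result's base color toward grey feeds nothing.
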MaterialFunction /Game/Characters/Player/Female/Medium/Bodies/F_MED_SnowKnight_Helm/Materials/MF_SnowKnight_AdvMats
graph {nodes [
  Constant() {r=1}
  FunctionOutput(None=NamedRerouteUsage())
  NamedRerouteDeclaration(None=FunctionInput() {inputName=Material, inputType=FunctionInput_MaterialAttributes}) {name=InMaterial, nodeColor=<rgba 0.1484 0 1 1>}
  n1 = FunctionInput() {bUsePreviewValueAsDefault=true, inputName=IceEdge, inputType=FunctionInput_Scalar, sortPriority=10, previewValue=<1 0 0 1>}
  NamedRerouteDeclaration(None=@n1) {name=IceEdge, nodeColor=<rgba 0.6172 1 0 1>}
  n2 = MaterialFunctionCall!(
   In=MaterialFunctionCall!(Material=NamedRerouteUsage(), EdgeMap=NamedRerouteUsage(), Alpha=NamedRerouteUsage()) {materialFunction=&/MF_SnowKnight_Ice}.Result
   Warp Mask
   Alpha=NamedRerouteUsage()
  ) {materialFunction=&/MF_SnowKnight_Galaxy}
  NamedRerouteDeclaration(None=MaterialFunctionCall!(Material Attributes=@n2.Result, Fresnel=NamedRerouteUsage(), Alpha=NamedRerouteUsage()) {materialFunction=/Game/Global/MaterialFunctions/TechArt/MF_ThinFilm_SnowKnight}.Result) {name=SnowKnightAdvMat, nodeColor=<rgba 0 1 0.3203 1>}
  NamedRerouteDeclaration(None=FunctionInput() {bUsePreviewValueAsDefault=true, inputName=Fresnel, inputType=FunctionInput_Scalar, sortPriority=3}) {name=Fresnel, nodeColor=<rgba 1 0 0.7734 1>}
  NamedRerouteDeclaration(None=FunctionInput() {bUsePreviewValueAsDefault=true, inputName=PrimaryMask, inputType=FunctionInput_Scalar, sortPriority=1}) {name=PrimaryMask, nodeColor=<rgba 1 0.8203 0 1>}
  NamedRerouteDeclaration(None=FunctionInput() {bUsePreviewValueAsDefault=true, inputName=SecondaryMask, inputType=FunctionInput_Scalar, sortPriority=2}) {name=SecondaryMask, nodeColor=<rgba 0 1 0.7422 1>}
  n3 = Multiply(A=ScalarParameter() {parameterName=PrimaryIce}, B=NamedRerouteUsage())
  n4 = Multiply(A=ScalarParameter() {parameterName=SecondaryIce}, B=NamedRerouteUsage())
  NamedRerouteDeclaration(None=Saturate(None=Add(A=@n3, B=@n4))) {name=IceAlpha, nodeColor=<rgba 0.7578 0 1 1>}
  n5 = Multiply(A=ScalarParameter() {parameterName=PrimaryGalaxy}, B=NamedRerouteUsage())
  n6 = Multiply(A=ScalarParameter() {parameterName=SecondaryGalaxy}, B=NamedRerouteUsage())
  NamedRerouteDeclaration(None=Saturate(None=Add(A=@n5, B=@n6))) {name=GalaxyAlpha, nodeColor=<rgba 0 0.2969 1 1>}
  n7 = Multiply(A=ScalarParameter() {parameterName=PrimaryChroma}, B=NamedRerouteUsage())
  n8 = Multiply(A=ScalarParameter() {parameterName=SecondaryChroma}, B=NamedRerouteUsage())
  NamedRerouteDeclaration(None=Saturate(None=Add(A=@n7, B=@n8))) {name=ChromaAlpha, nodeColor=<rgba 0 0.2969 1 1>}
  n9 = GetMaterialAttributes(None=@n2.Result) {attributeGetTypes [#1]}
  SetMaterialAttributes!(
   MaterialAttributes=@n9.MaterialAttributes
   BaseColor=LinearInterpolate(A=@n9.BaseColor, B=Constant3Vector() {constant=<rgba 0.2 0.2 0.2 0>}, Alpha=NamedRerouteUsage())
  ) {attributeSetTypes [#1]}
 ]}
