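# Material function for the SnowKnight character that adds an optional cloth fuzz look to incoming material attributes, blended through a channel mask.
# A normal-versus-camera rim term, tinted and scaled by base color, forms the fuzz color.
# A quality switch adds that fuzz to base color on lower tiers, while Epic switches to the Cloth shading model with subsurface color, Substrate-fix roughness bias and Fresnel specular scaled by packed-mask green scattering.
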MaterialFunction /Game/Characters/Player/Female/Medium/Bodies/F_MED_SnowKnight_Helm/Materials/MF_SnowKnight_ClothFuzz
graph {nodes [
  FunctionInput() {inputName=Fresnel, inputType=FunctionInput_Scalar, sortPriority=2}
  n1 = NamedRerouteUsage()
  n2 = GetMaterialAttributes(None=FunctionInput() {inputName=Material, inputType=FunctionInput_MaterialAttributes}) {attributeGetTypes [#1, #2, #3]}
  n3 = ScalarParameter() {defaultValue=0.9, group=06_Cloth, parameterName=Cloth_BaseColorIntensity}
  n4 = VectorParameter() {group=06_Cloth, parameterName=ClothFuzzTint, defaultValue=<rgba 1 1 1 1>}
  n5 = PixelNormalWS()
  n6 = CameraVectorWS()
  n7 = Power(Base=OneMinus(None=DotProduct(A=@n5, B=@n6)), Exp=ScalarParameter() {defaultValue=4, group=06_Cloth, parameterName=ClothFuzzExponent})
  n8 = Saturate(None=Multiply(A=Clamp(None=@n7), B=ScalarParameter() {defaultValue=1, group=06_Cloth, parameterName=Fuzz Fresnel Blend}))
  n9 = Multiply(A=@n4.RGB, B=Add(A=Multiply(A=@n8) {constB=10}))
  n10 = ScalarParameter() {defaultValue=1, group=06_Cloth, parameterName=ClothFuzzMulitplier}
  n11 = Saturate(None=Multiply(A=Multiply(A=@n2.BaseColor, B=@n3), B=Multiply(A=@n9, B=@n10)))
  n12 = Multiply(A=ScalarParameter() {defaultValue=1, group=06_Cloth, parameterName=Standard Quality Cloth Fuzz Intensity Multiplier}, B=@n11)
  n13 = SetMaterialAttributes!(MaterialAttributes=NamedRerouteUsage(), BaseColor=Saturate(None=Add(A=@n2.BaseColor, B=@n12))) {attributeSetTypes [#1]}
  n14 = Multiply(A=@n2.BaseColor, B=ScalarParameter() {defaultValue=1, group=06_Cloth, parameterName=ClothSubsurfaceIntensity})
  n15 = ScalarParameter() {defaultValue=0.4, group=06_Cloth_SubstrateFix, parameterName=SubstrateClothFix_RoughnessBias, sliderMax=1, sortPriority=0}
  n16 = Saturate(None=Multiply(A=@n2.Roughness, B=ScalarParameter() {defaultValue=1, group=06_Cloth, parameterName=ClothRoughness}))
  n17 = ScalarParameter() {defaultValue=5, group=06_Cloth_SubstrateFix, parameterName=SubstrateClothFix_SpecularFresnelExponent, sortPriority=0}
  n18 = ScalarParameter() {defaultValue=0.3, group=06_Cloth_SubstrateFix, parameterName=SubstrateClothFix_SpecularFresnelBase, sortPriority=0}
  n19 = LinearInterpolate(
   A=ScalarParameter() {defaultValue=0.75, group=06_Cloth, parameterName=ClothScatteringMin}
   B=ScalarParameter() {defaultValue=1, group=06_Cloth, parameterName=ClothScatteringMax}
   Alpha=MaterialFunctionCall(Float3=FunctionInput() {inputName=Ao/Gmap/Skin, sortPriority=2}) {materialFunction=&BreakOutFloat3Components}.G
  )
  n20 = SetMaterialAttributes!(
   MaterialAttributes=NamedRerouteUsage()
   ShadingModel=ShadingModel() {shadingModel=MSM_Cloth}
   SubsurfaceColor=Saturate(None=Multiply(A=@n4.RGB, B=@n14))
   Roughness=LinearInterpolate(A=@n15, Alpha=Saturate(None=@n16))
   CustomData0=ScalarParameter() {defaultValue=1, group=06_Cloth, parameterName=ClothShaderIntensity}
   Opacity=Constant() {r=1}
   BaseColor=@n11
   Specular=Multiply(A=Fresnel(ExponentIn=@n17, BaseReflectFractionIn=@n18) {baseReflectFraction=0}, B=Saturate(None=@n19))
  ) {attributeSetTypes [
    #4
    #5
    #3
    #6
    #7
    #1
    #8
   ]}
  n21 = MaterialFunctionCall!(
   Low _ PC Low and Med=@n13
   "High _ XB1, PS4"=@n13
   "Epic _ PC High, XSX, PS5"=@n20
   --------
   "Mobile Low _ PC Perf, Android"=@n13
   Mobile Med _ Switch=@n13
  ) {materialFunction=/Game/Global/MaterialFunctions/TechArt/MF_Fortnite_QualitySwitch_MaterialAttributes}
  n22 = ChannelMaskParameter(None=FunctionInput() {bUsePreviewValueAsDefault=true, inputName=Alpha, sortPriority=3}) {group=05_Tech Art Masks, parameterName=ClothFuzzMaskChannel}
  FunctionOutput(None=StaticSwitchParameter(True=BlendMaterialAttributes(A=@n1, B=@n21.Result, Alpha=@n22.R), False=@n1) {group=04_Tech Art, parameterName=UseClothFuzz})
  NamedRerouteDeclaration(None=@n2.MaterialAttributes) {name=Material Attributes, nodeColor=<rgba 1 0.4922 0 1>}
 ]}
