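# Material function that builds the pre-FX base material for the female SnowKnight helm character, on top of the 2023 character parent functions.
# Parent texture inputs pass through brightness/scale balance and emissive, with named reroutes feeding cloth fuzz and then metal LUT stages.
# The packed mask splits into AO, GMap and Skin.
# Alongside the material it outputs that mask, Fresnel, a tech art mask, distance blend and stage-switched pre-skinned local position.
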MaterialFunction /Game/Characters/Player/Female/Medium/Bodies/F_MED_SnowKnight_Helm/Materials/MF_SnowKnight_PreFX
graph {
 properties {bExposeToLibrary=true}
 nodes [
  n1 = NamedRerouteUsage()
  n2 = MaterialFunctionCall!(Material=NamedRerouteUsage(), Ao/Gmap/Skin=NamedRerouteUsage(), Fresnel=NamedRerouteUsage(), Alpha=NamedRerouteUsage()) {materialFunction=&/MF_SnowKnight_ClothFuzz}
  FunctionOutput(None=NamedRerouteDeclaration(None=MaterialFunctionCall!(Alpha=@n1, Materials=@n2.Result) {materialFunction=/Game/Global/MaterialFunctions/TechArt/MF_MetalLUT_TechArt}.Result) {name=MaterialOut, nodeColor=<rgba 0.9688 1 0 1>})
  n3 = MaterialFunctionCall() {materialFunction=/Game/Global/Materials/Character_Parent_2023/Functions/MF_CharacterParent_2023_TextureInputs}
  n4 = MaterialFunctionCall!(Material=@n3.Result) {materialFunction=/Game/Packages/Fortress_SharedMaterials/Base_Material_Functions/CharacterBrightnessandScale_Balance}
  FunctionOutput(None=@n4.Distance Blend) {outputName=Distance Blend, sortPriority=4}
  FunctionOutput(None=@n3.M) {outputName=M, sortPriority=1}
  n5 = Constant() {r=1}
  FunctionOutput(None=NamedRerouteDeclaration(None=Fresnel(ExponentIn=@n5) {baseReflectFraction=0}) {name=Fresnel, nodeColor=<rgba 1 0.3047 0 1>}) {outputName=Fresnel, sortPriority=9}
  n6 = TextureSampleParameter2D() {
   group=05_Tech Art Masks
   parameterName=TechArtMask
   samplerType=SAMPLERTYPE_Masks
   sortPriority=0
   texture=&T_BlackMask
  }
  FunctionOutput(None=NamedRerouteDeclaration(None=@n6.RGBA) {name="Tech Art Mask (v4)", nodeColor=<rgba 1 0 0.2344 1>}) {outputName=Tech Art Mask, sortPriority=11}
  n7 = PreSkinnedPosition()
  n8 = VertexInterpolator(VS=@n7)
  FunctionOutput(None=NamedRerouteDeclaration(None=ShaderStageSwitch(PixelShader=@n8.PS, VertexShader=@n7)) {name=Pre Skinned Local Postition, nodeColor=<rgba 1 0.5156 0 1>}) {outputName=Pre skinned Local Position, sortPriority=2}
  n9 = MaterialFunctionCall(Float3=@n3.M) {desc="R = AO\r\nG = GMap\r\nB = Skin", materialFunction=&BreakOutFloat3Components}
  NamedRerouteDeclaration(None=@n9.R) {name=Ambient Occlusion, nodeColor=<rgba 0.875 0 1 1>}
  NamedRerouteDeclaration(None=@n9.G) {name=GMap, nodeColor=<rgba 0.6641 0 1 1>}
  NamedRerouteDeclaration(None=@n9.B) {name=Skin, nodeColor=<rgba 0.6641 0 1 1>}
  NamedRerouteDeclaration(None=GetMaterialAttributes(None=@n3.Result) {attributeGetTypes [#1]}.Normal) {name=Normal Map, nodeColor=<rgba 1 0.3281 0 1>}
  NamedRerouteDeclaration(None=@n3.M) {name=Mask, nodeColor=<rgba 0.3125 0 1 1>}
  NamedRerouteDeclaration(None=@n6.RGB) {name=Tech Art Mask, nodeColor=<rgba 1 0.7031 0 1>}
  NamedRerouteDeclaration(None=MaterialFunctionCall!(Fresnel=NamedRerouteUsage(), Material=@n4.Result) {materialFunction=/Game/Global/Materials/Character_Parent_2023/Functions/MF_CharacterParent_2023_Emissive}.Result) {name=Material, nodeColor=<rgba 1 0 0.9844 1>}
 ]
}
