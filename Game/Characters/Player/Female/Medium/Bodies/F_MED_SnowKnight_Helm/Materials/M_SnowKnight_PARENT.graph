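# Parent material for the Snow Knight player character, built on material attributes with a per-expression shading model and usable on skinned meshes, cloth and Niagara.
# A PreFX function supplies base attributes and masks to a color customization function that also takes banner logo and chainmail masks.
# Static switches optionally layer advanced materials, driven by customization mask, Fresnel and ice edge, then a crystal effect.
# The shared CharacterParent 2023 PostFX function finishes the chain.
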Material /Game/Characters/Player/Female/Medium/Bodies/F_MED_SnowKnight_Helm/Materials/M_SnowKnight_PARENT
graph {
 properties {
  bAutomaticallySetUsageInEditor=false
  bUseMaterialAttributes=true
  bUsedWithClothing=true
  bUsedWithMeshDeformer=true
  bUsedWithMorphTargets=true
  bUsedWithNiagaraMeshParticles=true
  bUsedWithNiagaraRibbons=true
  bUsedWithNiagaraSprites=true
  bUsedWithSkeletalMesh=true
  pixelDepthOffsetMode=PDOM_Legacy
  shadingModel=MSM_FromMaterialExpression
  usedShadingModels=Default Lit | Subsurface Profile | Cloth
 }
 nodes [
  n1 = MaterialFunctionCall() {materialFunction=&/MF_SnowKnight_PreFX}
  n2 = MaterialFunctionCall!(
   MA=@n1.Result
   FX=NamedRerouteUsage()
   BannerMask=MaterialFunctionCall() {materialFunction=&/MFC_SnowKnight_BannerLogoMask}.BannerAlpha
   ChainmailMask=ComponentMask(None=NamedRerouteDeclaration(None=@n1.M) {name=M, nodeColor=<rgba 0.8516 0 1 1>}) {b=true}
  ) {materialFunction=&/MFC_SnowKnight_ColorCustomization}
  n3 = MaterialFunctionCall(Float4=@n2.Mask) {materialFunction=&BreakOutFloat4Components}
  n4 = StaticSwitch(
   True=MaterialFunctionCall!(
    Material=@n2.Result
    PrimaryMask=@n3.R
    SecondaryMask=@n3.R
    Fresnel=NamedRerouteUsage()
    IceEdge=ComponentMask(None=@n1.Tech Art Mask) {a=true}
   ) {materialFunction=&/MF_SnowKnight_AdvMats}.Result
   False=@n2.Result
   Value=StaticBoolParameter() {parameterName=UseAdvancedMats}
  )
  n5 = NamedRerouteUsage()
  MaterialFunctionCall(
   Material=StaticSwitchParameter(
    True=MaterialFunctionCall!(Material=@n4, Alpha=@n5) {materialFunction=&/MFC_SnowKnight_Crystal}.Result
    False=@n4
   ) {group=Skin FX, parameterName=Use Crystal, sortPriority=0}
   M=NamedRerouteUsage()
   Pre Skinned Local Position=NamedRerouteUsage()
   Distance Blend=NamedRerouteUsage()
   Fresnel=NamedRerouteUsage()
  ) {materialFunction=&MF_CharacterParent_2023_PostFX}
  NamedRerouteDeclaration(None=ComponentMask(None=@n1.Tech Art Mask) {b=true, g=true, r=true}) {name=TechArt/FX, nodeColor=<rgba 0.0652 0.0652 0.9635 1>}
  NamedRerouteDeclaration(None=@n3.A) {name=MaskCrystal, nodeColor=<rgba 0.0652 0.0652 0.9635 1>}
  NamedRerouteDeclaration(None=@n1.Fresnel) {name=Fresnel, nodeColor=<rgba 0 0.4609 1 1>}
  NamedRerouteDeclaration(None=@n3.R) {name=DEbug, nodeColor=<rgba 1 0 0.7734 1>}
  NamedRerouteDeclaration(None=@n1.Pre skinned Local Position) {name=PreskinnedPos, nodeColor=<rgba 1 0 0.8906 1>}
  NamedRerouteDeclaration(None=@n1.Distance Blend) {name=DistanceBlend, nodeColor=<rgba 1 0.8906 0 1>}
  OneMinus(None=ComponentMask(None=NamedRerouteUsage()) {r=true})
 ]
}
